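# Physics asset for the Nebula Axe pickaxe mesh, a cosmetic melee harvesting tool.
# A single capsule body on the b_AR_Root bone approximates the axe, offset along the bone and rotated to lie lengthwise.
# It has no constraints, so the weapon simulates and collides as one rigid piece.
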PhysicsAsset /BRCosmetics/Weapons/FORT_Melee/Pickaxe_Nebula/Meshes/Nebula_Axe_PhysicsAsset
bodies {b_AR_Root [{shapeType=Capsule, length=57.3812, radius=6.6859, center=<1.5707 0.0012 38.7455>, rotation=<rot 0.0035 -90 1.491>}]}
constraints []
